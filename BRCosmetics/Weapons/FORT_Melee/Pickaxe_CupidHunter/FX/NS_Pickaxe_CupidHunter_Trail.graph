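# Niagara swing trail effect for the CupidHunter pickaxe cosmetic, layering ribbon trails with heart-themed and glowing sprites.
# TipTrail001, FillTrail and HeartTrail use shared AnimTrail functions on skeletal mesh bones or sockets, fading and narrowing over particle life.
# Looping Bokeh, Dots and Heart emitters spawn sprites on the mesh; the first two drift with curl noise and inherited velocity.
# A time-of-day multiplier tints FillTrail, Heart and the dual-renderer tip ribbon.
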NiagaraSystem /BRCosmetics/Weapons/FORT_Melee/Pickaxe_CupidHunter/FX/NS_Pickaxe_CupidHunter_Trail
dependencies {
 usedDataInterfaces [&NiagaraDataInterfaceCurve, &NiagaraDataInterfaceColorCurve, &NiagaraDataInterfaceSkeletalMesh]
 usedDynamicInputs [&SimulationPosition]
 usedModules [
  &self:SystemSpawnScript.SystemScriptSource.SystemScriptGraph.NiagaraNodeAssignment_0.SetVariables_#1
  &SystemState
  &self:SystemSpawnScript.SystemScriptSource.SystemScriptGraph.NiagaraNodeAssignment_1.SetVariables_#2
 ]
}
emitters {
 Bokeh {
  emitterUpdateScript [
   &EmitterState(Life Cycle Mode=::NewEnumerator1, Loop Behavior=::NewEnumerator1, Loop Delay=0.05, Loop Duration=1, UseLoopDelay=true)
   &SpawnRate(SpawnRate=30, Use Spawn Probability=false)
  ]
  particleSpawnScript [
   &InitializeParticle(
    AdjustAlpha=false
    AdjustHue=true
    AdjustSaturation=false
    AdjustValue=false
    Color=<rgba 1 0.2737 0.0425 0.1>
    Color Mode=::NewEnumerator3
    Lifetime Max=0.7
    Lifetime Min=0.3
    Lifetime Mode=::NewEnumerator1
    Sprite Size Mode=::NewEnumerator1
    Uniform Sprite Size Max=20
    Uniform Sprite Size Min=15
    UsePositionOffset=false
    Write Material Random=true
   )
   &SkeletalMeshLocation(MaskSpawnGroup=false, Mesh Sampling Type=::NewEnumerator5, Skeletal Mesh=$System.Skeletal Mesh, Socket Sampling Mode=::NewEnumerator5)
   &ShapeLocation(MaskSpawnGroup=false, Override Seed=false, Sphere Radius=12, UseOwnerScale=true)
   &InheritVelocity(Inherited Velocity Speed Limit=160, Speed Limit=true, UseSpeedThreshold=false)
  ]
  particleUpdateScript [
   &CurlNoiseForce(Noise Frequency=15, Noise Strength=&RandomRangeFloat(Minimum=60, Maximum=120), Pan Noise=false)
   &Drag(Drag=5, Use Linear Drag=true, Use Rotational Drag=true)
   &SolveForcesAndVelocity(Clamp Velocity=false, Limit Acceleration=false)
   &ParticleState()
   &ScaleSpriteSize(Uniform Curve Sprite Scale=DataInterface() {propertyValues {Curve {keys [{interpMode=RCIM_Cubic, leaveTangent=2.6163, tangentMode=RCTM_Break, time=0, value=0}, {tangentMode=RCTM_User, time=1, value=1}]}}})
   &ScaleColor(
    Scale Alpha=&FloatFromCurve(FloatCurve=DataInterface() {propertyValues {Curve {keys [{interpMode=RCIM_Cubic, leaveTangent=-1.495, tangentMode=RCTM_Break, time=0, value=1}, <at 1 0>]}}})
    ScaleA=true
    ScaleRGB=true
   )
  ]
  rendererClasses [&NiagaraSpriteRendererProperties]
  renderers [{rendererClass=&NiagaraSpriteRendererProperties}]
 }
 Dots {
  emitterUpdateScript [
   &EmitterState(Life Cycle Mode=::NewEnumerator1, Loop Behavior=::NewEnumerator1, Loop Delay=0.05, Loop Duration=1, UseLoopDelay=true)
   &SpawnRate(SpawnRate=120, Use Spawn Probability=false)
  ]
  particleSpawnScript [
   &InitializeParticle(
    AdjustAlpha=false
    AdjustHue=true
    AdjustSaturation=false
    AdjustValue=false
    Color=<rgba 5 1.3685 0.2123 1>
    Color Mode=::NewEnumerator3
    Lifetime Max=0.4
    Lifetime Min=0.25
    Lifetime Mode=::NewEnumerator1
    Sprite Size Mode=::NewEnumerator1
    Uniform Sprite Size Max=1
    Uniform Sprite Size Min=0.6
    UsePositionOffset=false
    Write Material Random=true
   )
   &SkeletalMeshLocation(MaskSpawnGroup=false, Mesh Sampling Type=::NewEnumerator5, Skeletal Mesh=$System.Skeletal Mesh, Socket Sampling Mode=::NewEnumerator5)
   &ShapeLocation(MaskSpawnGroup=false, Override Seed=false, Sphere Radius=12, UseOwnerScale=true)
   &InheritVelocity(Inherited Velocity Speed Limit=160, Speed Limit=true, UseSpeedThreshold=false)
  ]
  particleUpdateScript [
   &CurlNoiseForce(Noise Frequency=15, Noise Strength=&RandomRangeFloat(Minimum=60, Maximum=120), Pan Noise=false)
   &Drag(Drag=5, Use Linear Drag=true, Use Rotational Drag=true)
   &SolveForcesAndVelocity(Clamp Velocity=false, Limit Acceleration=false)
   &ParticleState()
   &ScaleSpriteSize(
    Scale Sprite Size Mode=::NewEnumerator0
    Uniform Scale Factor=&Waveform(Global Phase=$Particles.NormalizedAge, "Waveform [1]"=::NewEnumerator2, "[1] Frequency"=5, "[1] Amplitude Min/Max"=<0 1>)
   )
  ]
  rendererClasses [&NiagaraSpriteRendererProperties]
  renderers [{rendererClass=&NiagaraSpriteRendererProperties}]
 }
 FillTrail {
  emitterSpawnScript [&FN_RibbonIDOffsetUtility()]
  emitterUpdateScript [
   &EmitterState()
   &FN_SpawnAnimTrail(Number of Anim Trails=2)
  ]
  particleSpawnScript [
   &FN_InitializeAnimTrailBoneIndecies(AnimTrail Selection=1, First Bone or Socket Index=1, Second Bone or Socket Index=2)
   &FN_InitializeAnimTrailBoneIndecies(First Bone or Socket Index=2) {module=FN_InitializeAnimTrailBoneIndecies001}
   &FN_InitializeAnimTrail(Color=<rgba 0.4531 0.4531 0.4531 1>, Lifetime=0.3, SkeletalMesh=$System.Skeletal Mesh)
  ]
  particleUpdateScript [
   &FN_UpdateAnimTrail(Locked Particle RibbonWidth=1, SkeletalMesh=$System.Skeletal Mesh) {module=FN_UpdateAnimTrail001}
   &ScaleColor(
    Scale Alpha=&FloatFromCurve(
     FloatCurve=DataInterface() {propertyValues {Curve {keys [
         {interpMode=RCIM_Cubic, leaveTangent=-7.3242, tangentMode=RCTM_Break, time=0, value=1}
         {arriveTangent=-0.1476, tangentMode=RCTM_User, time=0.55, value=0}
        ]}}}
     Scale Curve=0.5
    )
    Scale RGB=&MakeVectorFromLinearColorRGB(Color=&ColorFromCurve(ColorCurve=DataInterface() {propertyValues {
        AlphaCurve {keys []}
        BlueCurve {keys [<at 0 1.2 RCIM_Cubic>,<at 0.2127 0.625>,<at 0.5514 0.3229 RCIM_Cubic>]}
        GreenCurve {keys [<at 0 1.2 RCIM_Cubic>,<at 0.2127 0.625>,<at 0.5514 0.3229 RCIM_Cubic>]}
        RedCurve {keys [<at 0 1.2 RCIM_Cubic>,<at 0.2127 0.625>,<at 0.5514 0.3229 RCIM_Cubic>]}
       }}))
    ScaleA=true
    ScaleRGB=false
   )
   &ScaleRibbonWidth(Ribbon Width Scale=&FloatFromCurve(FloatCurve=DataInterface() {propertyValues {Curve {keys [{interpMode=RCIM_Cubic, leaveTangent=-5.5496, tangentMode=RCTM_Break, time=0.2, value=1}, <at 0.8 0.1>]}}}))
   &DynamicMaterialParameters(
    Index 0 Param 1=&FloatFromCurve(FloatCurve=DataInterface() {propertyValues {Curve {keys [<at 0 1>,<at 1 0>]}}})
    Index 0 Param 2=0.3
    Index 0 Param 3=2
    Index 0 Param 4=1
    Module.Index0Param0Write=true
    Module.Index0Param1Write=true
    Module.Index0Param2Write=true
    Module.Index0Param3Write=true
   )
   &ScaleColor(
    Scale Mode=::NewEnumerator1
    Scale RGBA=&MakeVector4_FromLinearColor(LinearColor=$System.TODColorMultiplier)
    ScaleRGBA=true
   ) {module=ScaleColor001}
  ]
  rendererClasses [&NiagaraRibbonRendererProperties]
  renderers [{rendererClass=&NiagaraRibbonRendererProperties}]
 }
 Heart {
  emitterUpdateScript [
   &EmitterState(Life Cycle Mode=::NewEnumerator1, Loop Behavior=::NewEnumerator1, Loop Delay=0.1, Loop Duration=0.5, UseLoopDelay=true)
   &SpawnRate(SpawnRate=80, Use Spawn Probability=false)
  ]
  particleSpawnScript [
   &InitializeParticle(
    AdjustAlpha=true
    AdjustHue=false
    AdjustSaturation=false
    AdjustValue=true
    Color Mode=::NewEnumerator3
    Hue Shift Range:Vector2f=<0 0>
    Lifetime=0.25
    Mass Mode=::NewEnumerator1
    Mesh Scale=&UniformRangedVector(Minimum=<1.25 1.25 1.25>, Maximum=<1.4 1.4 1.4>)
    Mesh Scale Mode=::NewEnumerator4
    Sprite Size=&Multiply_Vector2D_ByFloat(Float=&RandomRangeFloat(Minimum=0.06, Maximum=0.15), Vector2D=<180 180>)
    Sprite Size Mode=::NewEnumerator4
    UsePositionOffset=false
    Write Material Random=true
   ) {module=InitializeParticle001}
   &SkeletalMeshLocation(
    MaskSpawnGroup=false
    Mesh Sampling Type=::NewEnumerator5
    Orientation Sampling=::NewEnumerator1
    Skeletal Mesh=$System.Skeletal Mesh
    Socket Sampling Mode=::NewEnumerator5
   )
   &CameraOffset(Camera Offset Amount=15)
   /SetVariables(Particles.ImpactColor:LinearColor=<rgba 1 0 0.3994 1>) {isSetParameters=true}
  ]
  particleUpdateScript [
   &ParticleState()
   &DynamicMaterialParameters(
    Index 0 Param 1=&FloatFromCurve(FloatCurve=DataInterface() {propertyValues {Curve {keys [
         {arriveTangent=0.008, interpMode=RCIM_Cubic, leaveTangent=0.008, tangentMode=RCTM_User, time=0.1, value=0.1}
         {arriveTangent=0.0306, interpMode=RCIM_Cubic, leaveTangent=0.0306, tangentMode=RCTM_User, time=1, value=1}
        ]}}})
    Index 0 Param 4=&Subtract_Float(
     A=&FloatFromCurve(FloatCurve=DataInterface() {propertyValues {Curve {keys [{tangentMode=RCTM_User, time=0, value=0}, {tangentMode=RCTM_User, time=1, value=1}]}}})
     B=&RandomRangeFloat(Minimum=0, Maximum=10)
    )
    Module.Param0WriteEnabled=true
    Module.Param1WriteEnabled=true
    Module.Param2WriteEnabled=true
    Module.Param3WriteEnabled=true
   )
   &ScaleSpriteSize(Uniform Curve Sprite Scale=DataInterface() {propertyValues {Curve {keys [
        {arriveTangent=2.4606, interpMode=RCIM_Cubic, leaveTangent=2.4606, tangentMode=RCTM_User, time=0, value=0}
        {interpMode=RCIM_Cubic, tangentMode=RCTM_User, time=1, value=1}
       ]}}})
   &ScaleColor(
    Scale Alpha=&FloatFromCurve(FloatCurve=DataInterface() {propertyValues {Curve {keys [<at 0 1>,<at 1 0>]}}})
    Scale RGB=&MakeVectorFromLinearColorRGB(Color=&ColorFromCurve(ColorCurve=DataInterface() {propertyValues {
        AlphaCurve {keys [<at 0 1 RCIM_Cubic>,<at 1 0 RCIM_Cubic>]}
        BlueCurve {keys [<at 0 2.9388 RCIM_Cubic>,<at 0.2254 0.618>,<at 0.4718 0.2919>,<at 0.8284 0.2344 RCIM_Cubic>]}
        GreenCurve {keys [<at 0 1.5266 RCIM_Cubic>,<at 0.2254 0.4924>,<at 0.4718 0.2833>,<at 0.8284 0.2317 RCIM_Cubic>]}
        RedCurve {keys [<at 0 10 RCIM_Cubic>,<at 0.2254 2>,<at 0.4718 0.798>,<at 0.8284 0.3958 RCIM_Cubic>]}
       }}))
    ScaleA=true
    ScaleRGB=true
   )
   &ScaleColor(
    Scale Mode=::NewEnumerator1
    Scale RGBA=&MakeVector4_FromLinearColor(LinearColor=$System.TODColorMultiplier)
    ScaleRGBA=true
   ) {module=ScaleColor001}
   &SolveForcesAndVelocity(Clamp Velocity=false, Limit Acceleration=false)
   /SetVariables(Particles.Velocity=&Multiply_Vector(A=$Particles.Velocity, B=<1 1 0>)) {isSetParameters=true}
  ]
  rendererClasses [&NiagaraSpriteRendererProperties]
  renderers [{rendererClass=&NiagaraSpriteRendererProperties}]
 }
 HeartTrail {
  emitterSpawnScript [&FN_RibbonIDOffsetUtility()]
  emitterUpdateScript [
   &EmitterState()
   &FN_SpawnAnimTrail()
  ]
  particleSpawnScript [
   &FN_InitializeAnimTrailBoneIndecies(First Bone or Socket Index=3, Second Bone or Socket Index=4)
   &FN_InitializeAnimTrail(AnimTrailWidth=2, Color=<rgba 1.2 0.0576 0.0767 0.5>, Lifetime=0.3, SkeletalMesh=$System.Skeletal Mesh)
  ]
  particleUpdateScript [
   &FN_UpdateAnimTrail(Locked Particle RibbonWidth=1, SkeletalMesh=$System.Skeletal Mesh) {module=FN_UpdateAnimTrail001}
   &ScaleColor(
    Scale Alpha=&FloatFromCurve(
     FloatCurve=DataInterface() {propertyValues {Curve {keys [
         {interpMode=RCIM_Cubic, leaveTangent=-4.5336, tangentMode=RCTM_Break, time=0, value=1}
         {arriveTangent=-0.1476, tangentMode=RCTM_User, time=0.55, value=0}
        ]}}}
     Scale Curve=0.35
    )
    Scale RGB=&MakeVectorFromLinearColorRGB(Color=&ColorFromCurve(ColorCurve=DataInterface() {propertyValues {
        AlphaCurve {keys []}
        BlueCurve {keys [<at 0.0167 0.1661 RCIM_Cubic>,<at 0.2127 0.0792>,<at 0.5514 0.0337 RCIM_Cubic>]}
        GreenCurve {keys [<at 0.0167 0.1661 RCIM_Cubic>,<at 0.2127 0.0636>,<at 0.5514 0 RCIM_Cubic>]}
        RedCurve {keys [<at 0.0167 3 RCIM_Cubic>,<at 0.2127 1>,<at 0.5514 0.5052 RCIM_Cubic>]}
       }}))
    ScaleA=true
    ScaleRGB=true
   )
   &ScaleRibbonWidth(Ribbon Width Scale=&FloatFromCurve(FloatCurve=DataInterface() {propertyValues {Curve {keys [{interpMode=RCIM_Cubic, leaveTangent=-5.5496, tangentMode=RCTM_Break, time=0.2, value=1}, <at 0.8 0.1>]}}}))
   &DynamicMaterialParameters(
    Index 0 Param 1=&FloatFromCurve(FloatCurve=DataInterface() {propertyValues {Curve {keys [<at 0 1>,<at 1 0>]}}})
    Index 0 Param 2=0.3
    Index 0 Param 3=2
    Index 0 Param 4=1
    Module.Index0Param0Write=true
    Module.Index0Param1Write=true
    Module.Index0Param2Write=true
    Module.Index0Param3Write=true
   )
  ]
  rendererClasses [&NiagaraRibbonRendererProperties]
  renderers [{rendererClass=&NiagaraRibbonRendererProperties}]
 }
 TipTrail001 {
  emitterUpdateScript [
   &EmitterState()
   &FN_SpawnAnimTrail(Spawn Interval=0)
  ]
  particleSpawnScript [
   &FN_InitializeAnimTrailBoneIndecies(Second Bone or Socket Index=0)
   &FN_InitializeAnimTrail(
    AnimTrailWidth=10
    Color=&FNTimeOfDayColor(Night Color=<rgba 0.1782 0.1846 1 1>, Morning Color=<rgba 1 1 1 1>, Day Color=<rgba 1 1 1 1>, Evening Color=<rgba 2 1.44 0.8992 1>)
    Lifetime=0.22
    SkeletalMesh=$System.Skeletal Mesh
   )
  ]
  particleUpdateScript [
   &FN_UpdateAnimTrail(SkeletalMesh=$System.Skeletal Mesh)
   &ScaleColor(
    Scale Alpha=&Multiply_Float(
     A=&Multiply_Float(
      A=&FloatFromCurve(
       FloatCurve=DataInterface() {propertyValues {Curve {keys [
           {arriveTangent=8.1618, interpMode=RCIM_Cubic, leaveTangent=8.1618, tangentMode=RCTM_User, time=0, value=0.5}
           {arriveTangent=-1.6414, interpMode=RCIM_Cubic, leaveTangent=-1.6414, tangentMode=RCTM_User, time=0.1, value=1}
           {interpMode=RCIM_Cubic, tangentMode=RCTM_User, time=1, value=0}
          ]}}}
       Scale Curve=0.5
      )
      B=&MakeFloatFromLinearColor(LinearColor=$System.TODColorMultiplier, Channel:ENiagaraLinearColor_Channels=::NewEnumerator3)
     )
     B=&FNTimeOfDayScalar(Night Scale=1, Morning Scale=1, Evening Scale=0.25)
    )
    Scale RGB=&Multiply_VectorByFloat(
     Vector=&Multiply_Vector(
      A=&MakeVectorFromLinearColorRGB(Color=&ColorFromCurve(ColorCurve=DataInterface() {propertyValues {
          AlphaCurve {keys [<at 0 1 RCIM_Cubic>,<at 1 1 RCIM_Cubic>]}
          BlueCurve {keys [
            <at 0.0021 1.8125>
            {arriveTangent=-2.8648, interpMode=RCIM_Cubic, leaveTangent=-2.8648, time=0.1976, value=0.5}
            <at 0.5501 0.2425>
            <at 1 0.1309>
           ]}
          GreenCurve {keys [
            <at 0.0021 0.75>
            {arriveTangent=-1.2089, interpMode=RCIM_Cubic, leaveTangent=-1.2089, time=0.1976, value=0.2}
            <at 0.5501 0.0875>
            <at 1 0.0519>
           ]}
          RedCurve {keys [<at 0.0021 5>, {arriveTangent=-6.8428, interpMode=RCIM_Cubic, leaveTangent=-6.8428, time=0.1976, value=2}, <at 0.5501 1.25>, <at 1 1>]}
         }}))
      B=&MakeVectorFromLinearColorRGB(Color=$System.TODColorMultiplier)
     )
     Float=&FNTimeOfDayScalar(Night Scale=1, Morning Scale=1, Evening Scale=0.5)
    )
    ScaleA=true
    ScaleRGB=true
   )
   &ParticleState()
   /SetVariables(Particles.RibbonU0Override=&Multiply_Float(A=$Engine.Time, B=4)) {isSetParameters=true}
   &CameraOffset(Camera Offset Amount=10)
   &ScaleRibbonWidth(Ribbon Width Scale=&FloatFromCurve(FloatCurve=DataInterface() {propertyValues {Curve {keys [
         {arriveTangent=5.6298, interpMode=RCIM_Cubic, leaveTangent=5.6298, tangentMode=RCTM_User, time=0, value=0.6}
         {arriveTangent=-0.35, interpMode=RCIM_Cubic, leaveTangent=-0.35, time=0.1, value=1}
         {arriveTangent=-0.423, tangentMode=RCTM_User, time=1, value=0.25}
        ]}}}))
   /SetVariables(
    Particles.SecondaryColor:LinearColor=&MultiplyLinearColor(
     Linear Color=<rgba 0.2188 0.2188 0.2188 0.4>
     Scale Factor=&MakeVector4_FromLinearColor(LinearColor=$System.TODColorMultiplier)
    )
    Particles.SecondaryWidth:NiagaraFloat=&Multiply_Float(A=$Particles.RibbonWidth, B=1.1)
   ) {isSetParameters=true}
  ]
  rendererClasses [&NiagaraRibbonRendererProperties]
  renderers [
   {rendererClass=&NiagaraRibbonRendererProperties}
   {rendererClass=&NiagaraRibbonRendererProperties, rendererIndex=1}
  ]
 }
}
summary {
 systemName=NS_Pickaxe_CupidHunter_Trail
 emitters [
  {emitterName=Dots, rendererClasses [&NiagaraSpriteRendererProperties]}
  {emitterName=FillTrail, rendererClasses [&NiagaraRibbonRendererProperties]}
  {emitterName=Bokeh, rendererClasses [&NiagaraSpriteRendererProperties]}
  {emitterName=HeartTrail, rendererClasses [&NiagaraRibbonRendererProperties]}
  {emitterName=Heart, rendererClasses [&NiagaraSpriteRendererProperties]}
  {emitterName=TipTrail001, rendererClasses [&NiagaraRibbonRendererProperties]}
 ]
 userVariables [
  User.RibbonIDOffset {type=NiagaraInt32}
  User.SwingSectionIndex {type=NiagaraInt32}
 ]
}
