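# Physics asset for the CupidHunter two-handed pickaxe melee weapon's axe mesh.
# It defines a single capsule body on the Melee_TwoHanded_Handle bone, offset and rotated relative to that bone.
# With only one body, it has no constraints, so the weapon collides and simulates as one rigid piece.
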PhysicsAsset /BRCosmetics/Weapons/FORT_Melee/Pickaxe_CupidHunter/Meshes/CupidHunter_Axe_PhysicsAsset
bodies {Melee_TwoHanded_Handle [{shapeType=Capsule, length=78.4474, radius=18.6357, center=<0.8166 -5.6157 32.7978>, rotation=<rot 0.2321 -90 0.0004>}]}
constraints []
